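# Physics asset for the Pickaxe_Raider_Pink_Female_1h melee cosmetic, providing bodies and joints for its dangling ribbons.
# A capsule on the handle bone connects to a tiny ribbon-base capsule through a joint with free swing and twist.
# Four ribbon chains (A to D) branch from that base, each three box bodies long.
# Every ribbon joint shares the same asymmetric swing and tight twist limit values.
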PhysicsAsset /BRCosmetics/Weapons/FORT_Melee/Pickaxe_Raider_Pink_Female_1h/Meshes/Raider_Pink_Female_1h_Physics
bodies {
 Melee_TwoHanded_Handle [{shapeType=Capsule, length=1.1354, radius=11.5318, center=<0 -3.2613 0>, rotation=<rot 0 0 90>}]
 dyn_ribbon_A_01 [{shapeType=Box, extentX=12, extentY=12.5, extentZ=6, center=<5 0 1>}]
 dyn_ribbon_A_02 [{shapeType=Box, extentX=12, extentY=19, extentZ=6, center=<5.1577 0 -0.8945>, rotation=<rot 9.9999 0 0>}]
 dyn_ribbon_A_03 [{shapeType=Box, extentX=12, extentY=19, extentZ=6, center=<5 0 0>}]
 dyn_ribbon_B_01 [{shapeType=Box, extentX=12, extentY=12.5, extentZ=6, center=<5 0 1>}]
 dyn_ribbon_B_02 [{shapeType=Box, extentX=12, extentY=19, extentZ=6, center=<5.2072 0 -0.7731>, rotation=<rot 10 0 0>}]
 dyn_ribbon_B_03 [{shapeType=Box, extentX=12, extentY=19, extentZ=6, center=<5 0 0>}]
 dyn_ribbon_C_01 [{shapeType=Box, extentX=12, extentY=12.5, extentZ=6, center=<5 0 1>}]
 dyn_ribbon_C_02 [{shapeType=Box, extentX=12, extentY=19, extentZ=6, center=<5.1136 0 -0.6444>, rotation=<rot 9.9999 0 0>}]
 dyn_ribbon_C_03 [{shapeType=Box, extentX=12, extentY=19, extentZ=6, center=<5 0 0>}]
 dyn_ribbon_D_01 [{shapeType=Box, extentX=12, extentY=12.5, extentZ=6, center=<5 0 1>}]
 dyn_ribbon_D_02 [{shapeType=Box, extentX=12, extentY=19, extentZ=6, center=<5.2727 0.0001 -1.1387>, rotation=<rot 10 0 0>}]
 dyn_ribbon_D_03 [{shapeType=Box, extentX=12, extentY=19, extentZ=6, center=<5 0 0>}]
 dyn_ribbon_base [{shapeType=Capsule, length=1, radius=1}]
}
constraints [
 {
  bone1Name=dyn_ribbon_base
  bone2Name=Melee_TwoHanded_Handle
  swing1LimitDegrees=45
  swing1Motion=Free
  swing2LimitDegrees=45
  swing2Motion=Free
  twistLimitDegrees=45
  twistMotion=Free
 }
 {bone1Name=dyn_ribbon_A_01, bone2Name=dyn_ribbon_base, swing1LimitDegrees=45, swing2LimitDegrees=20, twistLimitDegrees=5}
 {bone1Name=dyn_ribbon_A_02, bone2Name=dyn_ribbon_A_01, swing1LimitDegrees=45, swing2LimitDegrees=20, twistLimitDegrees=5}
 {bone1Name=dyn_ribbon_A_03, bone2Name=dyn_ribbon_A_02, swing1LimitDegrees=45, swing2LimitDegrees=20, twistLimitDegrees=5}
 {bone1Name=dyn_ribbon_B_01, bone2Name=dyn_ribbon_base, swing1LimitDegrees=45, swing2LimitDegrees=20, twistLimitDegrees=5}
 {bone1Name=dyn_ribbon_B_02, bone2Name=dyn_ribbon_B_01, swing1LimitDegrees=45, swing2LimitDegrees=20, twistLimitDegrees=5}
 {bone1Name=dyn_ribbon_B_03, bone2Name=dyn_ribbon_B_02, swing1LimitDegrees=45, swing2LimitDegrees=20, twistLimitDegrees=5}
 {bone1Name=dyn_ribbon_C_01, bone2Name=dyn_ribbon_base, swing1LimitDegrees=45, swing2LimitDegrees=20, twistLimitDegrees=5}
 {bone1Name=dyn_ribbon_C_02, bone2Name=dyn_ribbon_C_01, swing1LimitDegrees=45, swing2LimitDegrees=20, twistLimitDegrees=5}
 {bone1Name=dyn_ribbon_C_03, bone2Name=dyn_ribbon_C_02, swing1LimitDegrees=45, swing2LimitDegrees=20, twistLimitDegrees=5}
 {bone1Name=dyn_ribbon_D_01, bone2Name=dyn_ribbon_base, swing1LimitDegrees=45, swing2LimitDegrees=20, twistLimitDegrees=5}
 {bone1Name=dyn_ribbon_D_02, bone2Name=dyn_ribbon_D_01, swing1LimitDegrees=45, swing2LimitDegrees=20, twistLimitDegrees=5}
 {bone1Name=dyn_ribbon_D_03, bone2Name=dyn_ribbon_D_02, swing1LimitDegrees=45, swing2LimitDegrees=20, twistLimitDegrees=5}
]
