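# Physics asset for the one-handed Pink Raider female pickaxe cosmetic, simulating a dangling hanging charm with a ball end.
# A capsule body on the handle anchors a three-link chain: two small capsules and a larger sphere on the dyn_hangings bones.
# The first link has tight swing and twist limits against the handle, while the lower links swing freely with modest twist limits, giving a pendulum-like sway.
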PhysicsAsset /BRCosmetics/Weapons/FORT_Melee/Pickaxe_Raider_Pink_Female_1h/Meshes/Raider_Pink_Female_1h_dyn_ball
bodies {
 Melee_TwoHanded_Handle [{shapeType=Capsule, length=36.6354, radius=5.7818, center=<0 0 20.0443>}]
 dyn_hangings_01 [{shapeType=Capsule, length=2, radius=2, center=<2 0 0>, rotation=<rot 90 0 0>}]
 dyn_hangings_02 [{shapeType=Capsule, length=2, radius=2, center=<2 0 0>, rotation=<rot 90 0 0>}]
 dyn_hangings_03 [{shapeType=Sphere, radius=4, center=<3.6 0 0>}]
}
constraints [
 {bone1Name=dyn_hangings_01, bone2Name=Melee_TwoHanded_Handle, swing1LimitDegrees=45, swing2LimitDegrees=15, twistLimitDegrees=10}
 {
  bone1Name=dyn_hangings_02
  bone2Name=dyn_hangings_01
  swing1LimitDegrees=75
  swing1Motion=Free
  swing2LimitDegrees=75
  swing2Motion=Free
  twistLimitDegrees=15
 }
 {
  bone1Name=dyn_hangings_03
  bone2Name=dyn_hangings_02
  swing1LimitDegrees=75
  swing1Motion=Free
  swing2LimitDegrees=75
  swing2Motion=Free
  twistLimitDegrees=15
 }
]
