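# Animation Blueprint for the pink one-handed female Raider pickaxe, adding physics-driven secondary motion to its dangling parts.
# The input pose runs through a RigidBody node, a Modify Bone node that shrinks the ribbon base bone, then a second RigidBody node with a separate dyn_ball physics asset.
# Both simulations run relative to the root bone with custom gravity and clamped motion, the second more tightly.
# The EventGraph holds no logic.
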AnimBlueprint /BRCosmetics/Weapons/FORT_Melee/Pickaxe_Raider_Pink_Female_1h/Meshes/Raider_Pink_Female_1h_AnimBP
target_skeleton=&/Raider_Pink_Female_1h_Skeleton
graphs [
 AnimGraph {nodes [
   AnimGraphNode_SubInput_0 = Input Pose()
   n1 = RigidBody(ComponentPose=Local To Component(LocalPose=@AnimGraphNode_SubInput_0)) {
    bClampLinearTranslationLimitToRefPose=true
    bOverrideWorldGravity=true
    overridePhysicsAsset=&/Raider_Pink_Female_1h_Physics
    simulationSpace=BaseBoneSpace
    alphaBoolBlend {blendOption=Linear}
    baseBoneRef {boneName=b_AR_Root}
    componentAppliedLinearAccClamp=<3000 3000 3000>
    overrideWorldGravity {z=-500}
    simSpaceSettings {maxAngularAcceleration=80, maxLinearAcceleration=800, worldAlpha=1}
   }
   n2 = "Transform (Modify) Bone - Bone: dyn_ribbon_base"(ComponentPose=@n1, Alpha=1, Translation="0,0,0", Rotation="0,0,0", Scale="0.85,0.85,0.85") {scaleMode=BMM_Replace, alphaBoolBlend {blendOption=Linear}, boneToModify {boneName=dyn_ribbon_base}, scale=<0.85 0.85 0.85>}
   n3 = RigidBody(ComponentPose=@n2) {
    bClampLinearTranslationLimitToRefPose=true
    bOverrideWorldGravity=true
    overridePhysicsAsset=&/Raider_Pink_Female_1h_dyn_ball
    simulationSpace=BaseBoneSpace
    alphaBoolBlend {blendOption=Linear}
    baseBoneRef {boneName=b_AR_Root}
    componentAppliedLinearAccClamp=<2000 2000 2000>
    overrideWorldGravity {z=-500}
    simSpaceSettings {maxAngularAcceleration=10, maxAngularVelocity=10, maxLinearAcceleration=200, maxLinearVelocity=200, worldAlpha=1}
   }
   Output Pose(Result=Component To Local(ComponentPose=@n3)) {layerGroup=None}
  ]}
 EventGraph {nodes [
   TryGetPawnOwner()
   Event BlueprintUpdateAnimation()
  ]}
]
variables []
