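# Animation Blueprint for the female pink Raider one-handed pickaxe, adding secondary physics motion to its ribbons.
# The input pose runs through a RigidBody node, then four chained AnimDynamics nodes on ribbon bones A–D, simulated relative to b_AR_Root.
# The event graph resets dynamics shortly after jumping or falling ends and when emotes end.
# Swimming alphas are set but unused; two bone transforms and a second RigidBody are unconnected.
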AnimBlueprint /BRCosmetics/Weapons/FORT_Melee/Pickaxe_Raider_Pink_Female_1h/Meshes/Raider_Pink_Female_1h_AnimBP_02
target_skeleton=&/Raider_Pink_Female_1h_Skeleton
graphs [
 AnimGraph {nodes [
   "Transform (Modify) Bone - Bone: b_AR_Root"(Alpha=1, Translation="0,0,0", Rotation="0,0,0", Scale="1,1,1") {
    rotationMode=BMM_Replace
    scaleMode=BMM_Replace
    translationMode=BMM_Replace
    alphaBoolBlend {blendOption=Linear}
    boneToModify {boneName=b_AR_Root}
   }
   "Transform (Modify) Bone - Bone: dyn_ribbon_base"(Alpha=1, Translation="0,0,0", Rotation="0,0,0", Scale="0.95,0.95,0.95") {
    scaleMode=BMM_Additive
    scaleSpace=BCS_BoneSpace
    alphaBoolBlend {blendOption=Linear}
    boneToModify {boneName=dyn_ribbon_base}
    scale=<0.95 0.95 0.95>
   }
   RigidBody() {
    bClampLinearTranslationLimitToRefPose=true
    bOverrideWorldGravity=true
    overridePhysicsAsset=&/Raider_Pink_Female_1h_dyn_ball
    simulationSpace=BaseBoneSpace
    alphaBoolBlend {blendOption=Linear}
    baseBoneRef {boneName=b_AR_Root}
    componentAppliedLinearAccClamp=<2000 2000 2000>
    overrideWorldGravity {z=-500}
    simSpaceSettings {maxAngularAcceleration=10, maxAngularVelocity=10, maxLinearAcceleration=200, maxLinearVelocity=200, worldAlpha=1}
   }
   n1 = RigidBody(ComponentPose=Local To Component(LocalPose=Input Pose())) {
    bClampLinearTranslationLimitToRefPose=true
    bOverrideWorldGravity=true
    overridePhysicsAsset=&/Raider_Pink_Female_1h_Physics
    simulationSpace=BaseBoneSpace
    alphaBoolBlend {blendOption=Linear}
    baseBoneRef {boneName=b_AR_Root}
    componentAppliedLinearAccClamp=<3000 3000 3000>
    overrideWorldGravity {z=-980}
    simSpaceSettings {maxAngularAcceleration=100, maxLinearAcceleration=1000, worldAlpha=1}
   }
   n2 = Anim Dynamics - Bone: dyn_ribbon_A_02(ComponentPose=@n1, Alpha=1, ExternalForce="0,0,0") {
    bGravityOverrideInSimSpace=true
    bUseGravityOverride=true
    simulationSpace=BoneRelative
    alphaBoolBlend {blendOption=Linear}
    boundBone {boneName=dyn_ribbon_A_02}
    componentLinearAccScale=<0.5 0.5 0.5>
    gravityOverride {z=-980}
    physicsBodyDefinitions [{
      boundBone {boneName=dyn_ribbon_A_02}
      boxExtents=<10 10 10>
      constraintSetup {angularLimitsMax=<50 50 50>, angularLimitsMin=<-50 -50 -50>, angularTarget=<0 0 0>, linearAxesMax=<5 5 5>, linearAxesMin=<-5 -5 -5>}
      localJointOffset=<0 0 0>
     }]
    relativeSpaceBone {boneName=b_AR_Root}
   }
   n3 = Anim Dynamics - Bone: dyn_ribbon_B_02(ComponentPose=@n2, Alpha=1, ExternalForce="0,0,0") {
    bGravityOverrideInSimSpace=true
    bUseGravityOverride=true
    simulationSpace=BoneRelative
    alphaBoolBlend {blendOption=Linear}
    boundBone {boneName=dyn_ribbon_B_02}
    componentLinearAccScale=<0.5 0.5 0.5>
    gravityOverride {z=-980}
    physicsBodyDefinitions [{
      boundBone {boneName=dyn_ribbon_B_02}
      boxExtents=<10 10 10>
      constraintSetup {angularLimitsMax=<50 50 50>, angularLimitsMin=<-50 -50 -50>, angularTarget=<0 0 0>, linearAxesMax=<5 5 5>, linearAxesMin=<-5 -5 -5>}
      localJointOffset=<0 0 0>
     }]
    relativeSpaceBone {boneName=b_AR_Root}
   }
   n4 = Anim Dynamics - Bone: dyn_ribbon_C_02(ComponentPose=@n3, Alpha=1, ExternalForce="0,0,0") {
    bGravityOverrideInSimSpace=true
    bUseGravityOverride=true
    simulationSpace=BoneRelative
    alphaBoolBlend {blendOption=Linear}
    boundBone {boneName=dyn_ribbon_C_02}
    componentLinearAccScale=<0.5 0.5 0.5>
    gravityOverride {z=-980}
    physicsBodyDefinitions [{
      boundBone {boneName=dyn_ribbon_C_02}
      boxExtents=<10 10 10>
      constraintSetup {angularLimitsMax=<50 50 50>, angularLimitsMin=<-50 -50 -50>, angularTarget=<0 0 0>, linearAxesMax=<5 5 5>, linearAxesMin=<-5 -5 -5>}
      localJointOffset=<0 0 0>
     }]
    relativeSpaceBone {boneName=b_AR_Root}
   }
   n5 = Anim Dynamics - Bone: dyn_ribbon_D_02(ComponentPose=@n4, Alpha=1, ExternalForce="0,0,0") {
    bGravityOverrideInSimSpace=true
    bUseGravityOverride=true
    simulationSpace=BoneRelative
    alphaBoolBlend {blendOption=Linear}
    boundBone {boneName=dyn_ribbon_D_02}
    componentLinearAccScale=<0.5 0.5 0.5>
    gravityOverride {z=-980}
    physicsBodyDefinitions [{
      boundBone {boneName=dyn_ribbon_D_02}
      boxExtents=<10 10 10>
      constraintSetup {angularLimitsMax=<50 50 50>, angularLimitsMin=<-50 -50 -50>, angularTarget=<0 0 0>, linearAxesMax=<5 5 5>, linearAxesMin=<-5 -5 -5>}
      localJointOffset=<0 0 0>
     }]
    relativeSpaceBone {boneName=b_AR_Root}
   }
   Output Pose(Result=Component To Local(ComponentPose=@n5)) {layerGroup=None}
  ]}
 EventGraph {nodes [
   Event BlueprintUpdateAnimation()
   > n1 = Sequence()
   > n2 = Branch(Condition=Get bIsSurfaceSwimming())
   > Set swimmingAlpha(swimmingAlpha:double=1) -> (Output_Get:double)
   > Set normalAlpha(normalAlpha:double=0) -> (Output_Get:double)
   Set swimmingAlpha(execute=@n2.else, swimmingAlpha=0)
   > Set normalAlpha(normalAlpha=1)
   n3 = Branch(execute=@n1.then_1, Condition=OR Boolean(A=Get bIsJumping(), B=Get bIsFalling()))
   Do Once(execute=@n3.else, Reset=@n3.then, Start Closed=true)
   > Delay(Duration=0.2)
   > ResetDynamics(InTeleportType=TeleportPhysics)
   > ResetDynamics(InTeleportType=ResetPhysics)
   Event OnEndPlayingEmote() {is=title K2Node_Event}
   > ResetDynamics(InTeleportType=TeleportPhysics)
   > ResetDynamics(InTeleportType=ResetPhysics)
   Event OnParachuteOpened()
   > Delay(Duration=1.5)
   TryGetPawnOwner()
   Event OnEndSkydiving()
   Event OnBeginSkydiving()
  ]}
]
variables [
 normalAlpha {type=number}
 swimmingAlpha {type=number}
]
